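; PCG primitive graph that, as its name and parameters indicate, fills input areas with randomly scattered points.
; It tags the input "Input:0", then chains three subgraphs: inputs management (defaulting to force-surface processing), core process and output preparation.
; Sampling density, seed, count, extents, uniformity, surface rotation and transform ranges feed the core process.
; Graph-typed parameters make the input processing, fallback, core and output stages swappable.
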(graph Fill_Areas_Randomly
  (params (InputsProcessingGraph :type (SoftObject /Script/PCG.PCGGraphInterface) :default /PCGPrimitives/Primitives/Shared/InputsProcessing/InputsProcessing_ForceSurface.InputsProcessing_ForceSurface) (InputsFallbackGraph :type (SoftObject /Script/PCG.PCGGraphInterface) :default /PCGPrimitives/Primitives/_Template_/Primitives_InputsFallback.Primitives_InputsFallback) (CoreProcessGraph :type (SoftObject /Script/PCG.PCGGraphInterface) :default /PCGPrimitives/Primitives/Fill/Subgraphs/Fill_Areas_Randomly_CoreProcess.Fill_Areas_Randomly_CoreProcess) (PrepareOutputsGraph :type (SoftObject /Script/PCG.PCGGraphInterface) :default /PCGPrimitives/Primitives/_Template_/Primitives_PrepareOutputs.Primitives_PrepareOutputs) (PointsPerSquaredMeter :type Double :default 1e+01) (SamplingPointExtents :type Vector :default (1.0 1.0 1.0)) (Uniformity :type Double) (SamplingSeed :type Int :default 1924953654) (FinalCount :type Int :default -1) (OutputPointsExtents :type Vector :default (1e+01 1e+01 1e+01)) (SurfaceRotation :type Double :default 1.0) (PositionRange :type Vector) (RotationRange :type Rotator) (ScaleRange :type Vector))
  (inputs (PrimaryInput :type Any))
  (bind AddTags_0 (Add%20Tags :TagsToAdd "Input:0"))
  ;[0 0 0 0.5 32] Inputs
  (bind Subgraph_0 (Subgraph :Graph /PCGPrimitives/Primitives/_Template_/Subgraphs/Primitives_InputsManagement.Primitives_InputsManagement :InputsFallbackGraph InputsFallbackGraph :InputsProcessingGraph InputsProcessingGraph))
  ;[0 0 0 0.5 32] Core Process
  (bind Subgraph_3 (Subgraph :Graph /PCGPrimitives/Primitives/Fill/Subgraphs/Fill_Areas_Randomly_CoreProcess.Fill_Areas_Randomly_CoreProcess :FinalCount FinalCount :OutputPointsExtents OutputPointsExtents :PointsPerSquaredMeter PointsPerSquaredMeter :PositionRange PositionRange :RotationRange RotationRange :SamplingPointExtents SamplingPointExtents :SamplingSeed SamplingSeed :ScaleRange ScaleRange :SubgraphOverride-param CoreProcessGraph :SurfaceRotation SurfaceRotation :Uniformity Uniformity))
  ;[0 0 0 0.5 32] Prepare Output
  (bind Subgraph_4 (Subgraph :Graph /PCGPrimitives/Primitives/_Template_/Primitives_PrepareOutputs.Primitives_PrepareOutputs :SubgraphOverride-param PrepareOutputsGraph))
  (connect DefaultInputNode PrimaryInput AddTags_0 In)
  (connect AddTags_0 Out Subgraph_0 In)
  (connect Subgraph_0 Out Subgraph_3 In)
  (connect Subgraph_3 Out Subgraph_4 In)
  (output Out Subgraph_4))
