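; PCG primitive graph that fills input areas with fitting assets, by default treating inputs as polygons.
; It tags the input "Input:0", then chains input management, a core fill subgraph and output preparation.
; Exposed parameters drive the core's seed, point density and extents, overlap, 90-degree random rotation, spawning and assembly handling.
; Each stage takes graph parameters, so processing, fallback, core and output logic are swappable.
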(graph Fill_Areas_WithFittingAssets
  (params (InputsProcessingGraph :type (SoftObject /Script/PCG.PCGGraphInterface) :default /PCGPrimitives/Primitives/Shared/InputsProcessing/InputsProcessing_ForcePolygon.InputsProcessing_ForcePolygon) (InputsFallbackGraph :type (SoftObject /Script/PCG.PCGGraphInterface) :default /PCGPrimitives/Primitives/_Template_/Primitives_InputsFallback.Primitives_InputsFallback) (CoreProcessGraph :type (SoftObject /Script/PCG.PCGGraphInterface) :default /PCGPrimitives/Primitives/Fill/Subgraphs/Fill_Areas_WithFittingAssets_CoreProcess.Fill_Areas_WithFittingAssets_CoreProcess) (PrepareOutputsGraph :type (SoftObject /Script/PCG.PCGGraphInterface) :default /PCGPrimitives/Primitives/_Template_/Primitives_PrepareOutputs.Primitives_PrepareOutputs) (Assets :type (Array (SoftObject /Script/CoreUObject.Object))) (Seed :type Int) (Overlap :type Double :default 1.0) (SpawnAssets :type Bool :default true) (KeepAssembliesRoot :type Bool) (Random90DegRotation :type Bool :default true) (BaseFillPointsPerSquareMeters :type Double :default 1.0) (BaseFillPointExtents :type Double :default 1e+02))
  (inputs (PrimaryInput :type Any))
  (bind AddTags_1 (Add%20Tags :TagsToAdd "Input:0"))
  ;[0 0 0 0.5 32] Inputs
  (bind Subgraph_0 (Subgraph :Graph /PCGPrimitives/Primitives/_Template_/Subgraphs/Primitives_InputsManagement.Primitives_InputsManagement :InputsFallbackGraph InputsFallbackGraph :InputsProcessingGraph InputsProcessingGraph))
  ;[0 0 0 0.5 32] Core Process
  (bind Subgraph_3 (Subgraph :Graph /PCGPrimitives/Primitives/Fill/Subgraphs/Fill_Areas_WithFittingAssets_CoreProcess.Fill_Areas_WithFittingAssets_CoreProcess :Base_fill_point_extents BaseFillPointExtents :Base_fill_points_per_square_meters BaseFillPointsPerSquareMeters :Keep_Assemblies_as_assets KeepAssembliesRoot :NewProperty Seed :overlap Overlap :random_90_rotation Random90DegRotation :Seed-param Seed :Spawn_Assets SpawnAssets :SubgraphOverride-param CoreProcessGraph))
  ;[0 0 0 0.5 32] Prepare Output
  (bind Subgraph_4 (Subgraph :Graph /PCGPrimitives/Primitives/_Template_/Primitives_PrepareOutputs.Primitives_PrepareOutputs :SubgraphOverride-param PrepareOutputsGraph))
  (connect DefaultInputNode PrimaryInput AddTags_1 In)
  (connect AddTags_1 Out Subgraph_0 In)
  (connect Subgraph_0 Out Subgraph_3 In)
  (connect Subgraph_3 Out Subgraph_4 In)
  (output Out Subgraph_4))
